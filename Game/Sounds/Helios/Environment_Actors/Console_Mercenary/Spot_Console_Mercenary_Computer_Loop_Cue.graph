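# Looping ambient Sound Cue for a Mercenary console computer, layering a terminal hum with randomized computer beeps and one-shots.
# A mixer blends three layers, weighted heavily toward the looping communication hum.
# Two looping branches pick random one-shots with pitch variation, one running continuously and the other spaced by short random delays.
# It uses a dedicated attenuation asset and a concurrency override that prevents new instances at the limit.
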SoundCue /Game/Sounds/Helios/Environment_Actors/Console_Mercenary/Spot_Console_Mercenary_Computer_Loop_Cue
root=Mixer_5
nodes [
 {id=Delay_1, children [Modulator_2], properties {delayMax=1, delayMin=0.5}}
 {id=Looping_0, children [Modulator_1]}
 {id=Looping_1, children [Delay_1]}
 {id=Mixer_5, children [Looping_0,Looping_1,Modulator_0], properties {inputVolume [0.4,0.2,3]}}
 {id=Modulator_0, children [WavePlayer_0], properties {pitchMin=1}}
 {id=Modulator_1, children [Random_0], properties {pitchMax=1.15, volumeMax=1, volumeMin=1}}
 {id=Modulator_2, children [Random_2], properties {pitchMax=1.15, volumeMax=1, volumeMin=1}}
 {id=Random_0, children [WavePlayer_5,WavePlayer_6,WavePlayer_7,WavePlayer_8], properties {weights [1,1,1,1]}}
 {id=Random_2, children [WavePlayer_4, WavePlayer_10], properties {weights [1,1]}}
 {
  id=WavePlayer_0
  properties {bLooping=true, soundWaveAssetPtr=/Game/Sounds/Helios/Environment_Actors/ConsoleTerminals/Spot_Console_Terminal_Communication_Hum_Loop_04}
 }
 {id=WavePlayer_10, properties {soundWaveAssetPtr=/Game/Sounds/Apollo/Biomes/GovernmentSecretComplex/SecretComplex_Computer_Oneshot_07}}
 {id=WavePlayer_4, properties {soundWaveAssetPtr=/Game/Sounds/Apollo/Biomes/GovernmentSecretComplex/SecretComplex_Computer_Oneshot_01}}
 {id=WavePlayer_5, properties {soundWaveAssetPtr=/Game/Sounds/Apollo/Biomes/GovernmentSecretComplex/SecretComplex_Computer_Oneshot_02}}
 {id=WavePlayer_6, properties {soundWaveAssetPtr=/Game/Sounds/Apollo/Biomes/GovernmentSecretComplex/SecretComplex_Computer_Oneshot_03}}
 {id=WavePlayer_7, properties {soundWaveAssetPtr=/Game/Sounds/Apollo/Biomes/GovernmentSecretComplex/SecretComplex_Computer_Oneshot_04}}
 {id=WavePlayer_8, properties {soundWaveAssetPtr=/Game/Sounds/Apollo/Biomes/GovernmentSecretComplex/SecretComplex_Computer_Oneshot_05}}
]
properties {
 attenuationSettings=&/Spot_Console_Mercenary_Computer_Loop_Att
 bOverrideConcurrency=true
 maxDistance=625
 soundClassObject=&Fort_Gadgets_Items
 volumeMultiplier=0.2
 attenuationOverrides {falloffDistance=2000}
 concurrencyOverrides {resolutionRule=PreventNew}
 concurrencySet [&Weapon_Drop_StopOldest]
}
